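; Base boss pawn Blueprint for non-participant Jethro bosses, adding a first-hit health bar reaction to its parent pawn.
; BeginPlay calls the parent, then binds a custom handler to the OnDamaged dispatcher.
; The handler uses a DoOnce with no reset, so only the first damage event calls ForceHideHealthBar with a zero argument.
; The overlap and tick events are disabled stubs, and the construction script only calls the parent.
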(event EventBeginPlay
  (|CallParentFunction)
  (EventDispatchers|BindEventtoOnDamaged (delegate Custom|JethroBoss_OnDamaged)))

(event Custom|JethroBoss_OnDamaged
  (params (DamagedActor :type Actor%20Object%20Reference) (Damage :type Float%20%28single-precision%29) (InstigatedBy :type Controller%20Object%20Reference) (DamageCauser :type Actor%20Object%20Reference) (HitLocation :type Vector) (FHitComponent :type Primitive%20Component%20Object%20Reference) (BoneName :type Name) (Momentum :type Vector))
  (Utilities|FlowControl|DoOnce false
    (:Completed
      (CallFunction|ForceHideHealthBar 0.0))))

(event Collision|EventActorBeginOverlap
  (params (OtherActor :type Actor%20Object%20Reference))
  (|CallParentFunction OtherActor) ; This node is disabled and will not be called.\nDrag off pins to build functionality.
)

(event EventTick
  (params (DeltaSeconds :type Float%20%28single-precision%29))
  (|CallParentFunction DeltaSeconds) ; This node is disabled and will not be called.\nDrag off pins to build functionality.
)

(fn ConstructionScript
  (|CallParentFunction))
